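; Blueprint for the Apollo_ZP_Shell_01 actor in the Apollo Tower environment set, managing an ambient audio component.
; BeginPlay starts the AmbientAudio component from the beginning.
; On Destroyed, it fades that audio out briefly to silence with a linear curve, with the component reference bound before the fade.
; The construction script is empty.
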(event EventBeginPlay
  (Audio|Components|Audio|Play (Variables|Default|GetAmbientAudio) 0.0))

(event EventDestroyed
  (bind _ambientaudio (Variables|Default|GetAmbientAudio)
    (:then
      (Audio|Components|Audio|FadeOut _ambientaudio 0.2 0.0 "Linear" :!class /Script/Engine.AudioComponent))
    (:else)))

(fn ConstructionScript)
