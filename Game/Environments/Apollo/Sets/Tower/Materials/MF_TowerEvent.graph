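# Material function for the Apollo Tower set that offsets vertices along the object's local down axis for a timed tower event.
# MF_Timestamp_InOut_01 turns timestamp and duration from custom primitive data into a smoothstep blend, scaled by Z object bounds to set the travel distance.
# The TowerEvent static switch either replaces World Position Offset with this result or passes the input attributes through.
# A looping modulo-time chain is left unconnected.
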MaterialFunction /Game/Environments/Apollo/Sets/Tower/Materials/MF_TowerEvent
graph {nodes [
  n1 = FunctionInput() {inputType=FunctionInput_MaterialAttributes}
  n2 = GetMaterialAttributes(None=@n1) {attributeGetTypes [#1]}
  n3 = Constant()
  n4 = Constant() {r=0.1}
  n5 = ScalarParameter() {bUseCustomPrimitiveData=true, parameterName=Timestamp}
  n6 = ScalarParameter() {bUseCustomPrimitiveData=true, parameterName=TimestampDuration, primitiveDataIndex=1}
  n7 = StaticBool()
  n8 = Constant3Vector() {constant=<rgba 0 0 1 0>}
  n9 = ComponentMask(None=ObjectBounds()) {b=true}
  n10 = Multiply(
   A=LinearInterpolate(Alpha=MaterialFunctionCall(TimestampA=@n3, TransitionLength_In=@n4, TimestampB=@n5, TransitionLength_Out=@n6, Transition In Only?=@n7) {materialFunction=&MF_Timestamp_InOut_01}.SmoothStep) {constB=0.9}
   B=Multiply(A=Transform(None=@n8) {transformSourceType=TRANSFORMSOURCE_Local}, B=Multiply(A=@n9) {constB=-1.33})
  )
  FunctionOutput(None=StaticSwitchParameter(
    True=SetMaterialAttributes!(MaterialAttributes=@n2.MaterialAttributes, WorldPositionOffset=@n10) {attributeSetTypes [#1]}
    False=NamedRerouteUsage()
   ) {parameterName=TowerEvent})
  Multiply(A=MaterialFunctionCall!(Time, Divisor=Constant() {r=5}, Mod) {materialFunction=/Game/Effects/Fort_Effects/Materials/_Functions/MF_ModuloTime}.Clamped 0-1) {constB=-1}
  NamedRerouteDeclaration(None=@n1) {name=MA, nodeColor=<rgba 0 0.035 1 1>}
 ]}
