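# Master material for the Apollo small tower set, with static switches for ramp and ground variants.
# Height-masked bottom and fresnel-tinted top layers receive dust; the ground variant blends world-aligned tiles, dirt and cliff, with parallax occlusion at Epic quality.
# The result blends into landscape runtime virtual textures near terrain, with simpler shading-path and proxy fallbacks.
# Optional gradient, time-of-day glow, animated noise emissive and sine wing offset join MF_TowerEvent.
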Material /Game/Environments/Apollo/Sets/Tower/Materials/M_SmallTower_01
graph {
 properties {
  bTangentSpaceNormal=false
  bUseMaterialAttributes=true
  bUsedWithNanite=true
  bUsedWithNiagaraMeshParticles=true
  physMaterial=/Game/FortressPhysicalMaterials/Stone
  pixelDepthOffsetMode=PDOM_Legacy
  usedShadingModels=Default Lit
 }
 nodes [
  n1 = RuntimeVirtualTextureSampleParameter() {
   group=RVT
   materialType=WorldHeight
   parameterName=RVTHeight
   virtualTexture=/Game/Athena/Apollo/Maps/Landscape/VT/RVT_Apollo_Landscape_Height
  }
  NamedRerouteDeclaration(None=@n1.WorldHeight) {name=RVT_H, nodeColor=<rgba 1 0 0.75 1>}
  n2 = ComponentMask(None=WorldPosition().XYZ) {b=true}
  n3 = NamedRerouteUsage()
  n4 = Constant()
  n5 = SphereMask(A=@n2, B=@n3, Radius=ScalarParameter() {defaultValue=1000, group=Glow, parameterName=RVTFadeGlow}, Hardness=@n4)
  NamedRerouteDeclaration(None=MaterialFunctionCall!(NoRVT=Constant() {r=1}, RVT=OneMinus(None=@n5)) {materialFunction=/Game/Packages/Fortress_SharedMaterials/Base_Material_Functions/FN_RVT_Switch_Color}.Result) {name=RVTGlow, nodeColor=<rgba 1 0 0.1406 1>}
  NamedRerouteDeclaration(None=Multiply(A=TextureCoordinate(), B=ScalarParameter() {defaultValue=1, group=Base, parameterName=Scale UV, sortPriority=-1})) {name=UV Scale, nodeColor=<rgba 1 0.6094 0 1>}
  n6 = NamedRerouteUsage()
  n7 = TextureSampleParameter2D(UVs=@n6) {
   group=Base
   parameterName=Normal
   samplerSource=SSM_Wrap_WorldGroupSettings
   samplerType=SAMPLERTYPE_Normal
   sortPriority=2
   texture=/Game/Environments/Apollo/Sets/Tower/Textures/T_SmallTower_001_N
  }
  NamedRerouteDeclaration(None=@n7.RGB) {name=BasicNormal, nodeColor=<rgba 1 0.8438 0 1>}
  n8 = VectorParameter() {parameterName=TwPick_WPO_Speed, defaultValue=<rgba 0.5 0.3 0.29 2.1>}
  n9 = Multiply(A=Time(), B=Multiply(A=@n8.RGB, B=@n8.RGB))
  n10 = Add(A=ScalarParameter() {defaultValue=0.16, parameterName=TwPick_WPO_Speed_Offset}, B=@n9) {constA=0.1, constB=0.01}
  n11 = PreSkinnedPosition()
  n12 = PreSkinnedLocalBounds()
  n13 = Divide(A=Subtract(A=@n11, B=@n12.Min), B=Subtract(A=@n12.Max, B=@n12.Max))
  n14 = SphereMask(A=Add(A=ScalarParameter() {defaultValue=0.25, parameterName=TwPick_mask offset}, B=ComponentMask(None=@n13) {g=true}), B=Constant() {r=0.5}) {attenuationRadius=0.3}
  n15 = LinearInterpolate(
   A=VectorParameter() {parameterName=TwPick_WPO_Min, defaultValue=<rgba -1 1 -4 1>}.RGB
   B=VectorParameter() {parameterName=TwPick_WPO_Max, defaultValue=<rgba 1 1 3 1>}.RGB
   Alpha=Sine(None=LinearInterpolate(A=@n10, B=@n9, Alpha=@n14))
  )
  n16 = LinearInterpolate(A=Multiply(A=Constant3Vector() {constant=<rgba -1 1 1 0>}, B=@n15), B=@n15, Alpha=@n14)
  n17 = VertexNormalWS()
  n18 = SphereMask(
   A=@n2
   B=@n3
   Radius=LinearInterpolate(
    A=ScalarParameter() {defaultValue=255, group=RVT, parameterName=RVTFadeDistanceSides}
    B=ScalarParameter() {defaultValue=255, group=RVT, parameterName=RVTFadeDistance}
    Alpha=ComponentMask(None=@n17) {b=true}
   )
   Hardness=Constant()
  )
  n19 = Constant() {r=-1.5}
  n20 = Constant() {r=1}
  n21 = TextureSampleParameter2D(UVs=Multiply(A=TextureCoordinate(), B=ScalarParameter() {defaultValue=5, group=Base, parameterName=DetailsScale, sortPriority=4})) {
   group=Base
   parameterName=DetailsBottom
   samplerSource=SSM_Wrap_WorldGroupSettings
   sortPriority=4
   texture=/Game/Environments/Apollo/Rocks/Natural_Rocks/Textures/T_RockDetails_02D
  }
  n22 = VectorParameter() {group=Dust, parameterName=Dust, defaultValue=<rgba 0.276 0.1329 0.0701 0>}
  n23 = Desaturation(None=@n22.RGB, Fraction=Constant() {r=1})
  n24 = Saturate(None=OneMinus(None=NamedRerouteUsage()))
  n25 = TextureSampleParameter2D(UVs=Multiply(A=TextureCoordinate(), B=ScalarParameter() {defaultValue=5, group=Base, parameterName=DetailsScale, sortPriority=4})) {
   group=Base
   parameterName=DetailsBottom
   samplerSource=SSM_Wrap_WorldGroupSettings
   sortPriority=4
   texture=&T_RockDetails_02D
  }
  n26 = SetMaterialAttributes!(
   MaterialAttributes
   Base Color=Multiply(A=LinearInterpolate(A=@n22.RGB, B=@n23, Alpha=@n24), B=@n25.R)
   Normal=NamedRerouteUsage()
   Metallic
   Roughness=Constant() {r=1}
   Specular
  ) {attributeSetTypes [
    #1
    #2
    #3
    #4
    #5
   ]}
  n27 = Multiply(
   A=MaterialFunctionCall(In=NamedRerouteUsage(), Power=ScalarParameter() {defaultValue=4, group=Base, parameterName=Curves Power}) {materialFunction=&SCurve}.Result
   B=ScalarParameter() {defaultValue=0.15, group=Base, parameterName=CurvesValue}
  )
  n28 = TextureSampleParameter2D(UVs=@n6) {
   group=Base
   parameterName=Mask
   samplerSource=SSM_Wrap_WorldGroupSettings
   samplerType=SAMPLERTYPE_Masks
   sortPriority=1
   texture=/Game/Environments/Apollo/Sets/Tower/Textures/T_SmallTower_001_MSK
  }
  n29 = Multiply(A=TextureCoordinate(), B=ScalarParameter() {defaultValue=5, group=Base, parameterName=DetailsScale, sortPriority=4})
  n30 = TextureSampleParameter2D(UVs=@n29) {
   group=Base
   parameterName=DetailsBottom
   samplerSource=SSM_Wrap_WorldGroupSettings
   sortPriority=4
   texture=&T_RockDetails_02D
  }
  n31 = MaterialFunctionCall(
   Base=Multiply(
    A=VectorParameter() {group=Base, parameterName=ToneBottom, sortPriority=-1, defaultValue=<rgba 0.2474 0.2408 0.2317 0>}.RGB
    B=NamedRerouteDeclaration(None=@n28.R) {name=Mask R, nodeColor=<rgba 0.3359 0 1 1>}
   )
   Blend=@n30.RGB
  ) {materialFunction=&Blend_Overlay}
  n32 = TextureSample(UVs=Multiply(A=TextureCoordinate(), B=Constant() {r=4})) {samplerType=SAMPLERTYPE_Normal, texture=/Game/Global/Textures/Default/Tileable_Patterns/T_IceScratch1_N}
  n33 = Constant() {r=0.035}
  n34 = Add(A=@n7.RGB, B=Multiply(A=Multiply(A=@n32.RGB, B=@n33), B=Constant3Vector() {constant=<rgba 1 1 0 0>}))
  n35 = Constant() {r=0.45}
  n36 = NamedRerouteDeclaration(None=@n28.G) {name=Mask G, nodeColor=<rgba 1 0 0.3047 1>}
  n37 = Constant() {r=2}
  n38 = TextureSampleParameter2D(UVs=@n29) {
   group=Base
   parameterName=DetailsBottom_N
   samplerSource=SSM_Wrap_WorldGroupSettings
   samplerType=SAMPLERTYPE_Normal
   sortPriority=3
   texture=/Game/Environments/Apollo/Rocks/Natural_Rocks/Textures/T_RockDetails_02_N
  }
  n39 = TextureSampleParameter2D(UVs=Multiply(A=@n29, B=Constant() {r=0.25})) {
   group=Base
   parameterName=DetailsBottom_N
   samplerSource=SSM_Wrap_WorldGroupSettings
   samplerType=SAMPLERTYPE_Normal
   sortPriority=3
   texture=&T_RockDetails_02_N
  }
  n40 = Add(
   A=NamedRerouteDeclaration(None=MaterialFunctionCall(A=@n34, B=@n7.RGB, Transition Phase=@n35, Height Texture=@n36, Contrast=@n37) {materialFunction=&HeightLerp}.Results) {name=Normal+Scratches, nodeColor=<rgba 0.7578 0 1 1>}
   B=Multiply(
    A=Add(A=@n38.RGB, B=@n39.RGB)
    B=VectorParameter() {group=Base, parameterName=DetailsBottom N Val, sortPriority=4, defaultValue=<rgba -0.5 -0.5 0 0>}.RGB
   )
  )
  n41 = ShadingPathSwitch(Default=@n40, Deferred=@n40, Forward=@n7.RGB, Mobile=@n7.RGB)
  n42 = SetMaterialAttributes!(
   MaterialAttributes
   Base Color=NamedRerouteDeclaration(None=Max(A=@n27, B=@n31.Result)) {name=BasicDiffuse, nodeColor=<rgba 0 1 0.0391 1>}
   Metallic=Constant()
   Specular=LinearInterpolate(
    A=ScalarParameter() {defaultValue=0.05, group=Base, parameterName=Spec Min Bottom, sortPriority=4}
    B=ScalarParameter() {defaultValue=0.2, group=Base, parameterName=Spec Max Bottom, sortPriority=4}
    Alpha=NamedRerouteUsage()
   )
   Roughness=LinearInterpolate(
    A=ScalarParameter() {defaultValue=-1.3473, group=Grou, parameterName=Rough Min Bottom, sortPriority=4}
    B=ScalarParameter() {defaultValue=1, group=Base, parameterName=Rough Max Bottom, sortPriority=4}
    Alpha=@n30.R
   )
   Normal=NamedRerouteDeclaration(None=Transform(None=@n41)) {name=Normals Bottom, nodeColor=<rgba 0.2656 1 0 1>}
  ) {attributeSetTypes [
    #1
    #3
    #5
    #4
    #2
   ]}
  n43 = GetMaterialAttributes(None=@n42) {attributeGetTypes [#1]}
  n44 = ScalarParameter() {defaultValue=-1024, group=Gradient, parameterName=Gradient Z Offset}
  n45 = ComponentMask(None=WorldPosition().XYZ) {b=true}
  n46 = Saturate(None=Divide(A=Add(A=@n44, B=@n45), B=ScalarParameter() {defaultValue=1800, group=Gradient, parameterName=Gradient Z Blur}))
  n47 = NamedRerouteUsage()
  n48 = TextureSample(UVs=Fresnel(ExponentIn=Constant() {r=1}, BaseReflectFractionIn=Constant() {r=3}, Normal=Transform(None=@n47)) {baseReflectFraction=0.5, exponent=3}) {texture=/Game/Global/Textures/Default/Masks_and_Gradients/3pxRGBImage}
  n49 = Multiply(A=Constant() {r=4}, B=@n48.RGB)
  n50 = DotProduct(A=AtmosphericLightVector(), B=ReflectionVectorWS())
  n51 = Constant() {r=55}
  n52 = Saturate(None=Multiply(A=Power(Base=@n50, Exp=@n51), B=Constant() {r=512}))
  n53 = Multiply(A=VectorParameter() {group=Top, parameterName=ToneTop, sortPriority=1, defaultValue=<rgba 0.6693 0.6693 0.6693 0>}.RGB, B=@n43.Base Color)
  n54 = Multiply(
   A=Multiply(A=LinearInterpolate(A=@n48.RGB, B=@n49, Alpha=@n52), B=@n53)
   B=StaticSwitchParameter(True=Constant() {r=0.1}, False=ScalarParameter() {defaultValue=0.5, group=Anodizing, parameterName=Anodizing, sortPriority=7}) {group=Ramp, parameterName=IsRamp}
  )
  n55 = StaticSwitchParameter(True=Multiply(A=NamedRerouteDeclaration(None=@n46) {name=GradientMask, nodeColor=<rgba 0 1 0.8828 1>}, B=@n54), False=@n54) {group=Base, parameterName=use Gradient, sortPriority=-4}
  n56 = NamedRerouteUsage()
  n57 = VectorParameter() {parameterName=HLOD_ColorMult, defaultValue=<rgba 1 1 1 1>}
  n58 = TextureSampleParameter2D(UVs=Multiply(A=TextureCoordinate(), B=ScalarParameter() {defaultValue=55, group=Top, parameterName=DetailScale_Top, sortPriority=2})) {
   group=Top
   parameterName=DetailsTop
   samplerType=SAMPLERTYPE_Normal
   sortPriority=1
   texture=/Game/Environments/Apollo/Sets/DesertPoint/Crystals/Textures/CrystalDetails_N
  }
  n59 = Add(
   A=NamedRerouteUsage()
   B=Multiply(A=@n58.RGB, B=VectorParameter() {group=Top, parameterName=DetailN_Top_Value, defaultValue=<rgba -0.5 -0.5 0 1>}.RGB)
  )
  n60 = NamedRerouteUsage()
  n61 = Transform(None=ShadingPathSwitch(Default=@n59, Deferred=@n59, Forward=@n60, Mobile=@n60))
  n62 = Fresnel(Normal=@n61)
  n63 = NamedRerouteUsage()
  n64 = ScalarParameter() {defaultValue=0.3, group=Top, parameterName=RoughnessTop1, sortPriority=5}
  n65 = TextureSampleParameter2D(UVs=Multiply(A=TextureCoordinate(), B=ScalarParameter() {defaultValue=5, group=Base, parameterName=DetailsScale, sortPriority=4})) {
   group=Base
   parameterName=DetailsBottom
   samplerSource=SSM_Wrap_WorldGroupSettings
   sortPriority=4
   texture=&T_RockDetails_02D
  }
  n66 = Multiply(A=@n65.R, B=Constant() {r=0.5})
  n67 = SetMaterialAttributes(
   MaterialAttributes=@n43.MaterialAttributes
   Base Color=MaterialProxyReplace(Realtime=Add(A=@n55, B=@n53), MaterialProxy=Multiply(A=@n56, B=@n57.RGB))
   Metallic=Multiply(A=Constant() {r=0.8}, B=OneMinus(None=@n62))
   Specular=LinearInterpolate(
    A=ScalarParameter() {defaultValue=0.05, group=Top, parameterName=Spec Min Top, sortPriority=3}
    B=ScalarParameter() {defaultValue=0.5, group=Top, parameterName=Spec Max Top, sortPriority=4}
    Alpha=@n63
   )
   Roughness=LinearInterpolate(A=Max(A=@n64, B=@n66), B=ScalarParameter() {defaultValue=0.2, group=Top, parameterName=RoughnessTop2, sortPriority=6}, Alpha=@n63)
   Normal=MaterialProxyReplace(Realtime=@n61, MaterialProxy=NamedRerouteUsage())
  ) {attributeSetTypes [
    #1
    #3
    #5
    #4
    #2
   ]}
  n68 = GetMaterialAttributes(None=@n67) {attributeGetTypes [#4]}
  n69 = Multiply(A=@n68.Roughness, B=ScalarParameter() {defaultValue=1.5, group=Ramp, parameterName=RampRoughnessMod})
  n70 = ComponentMask(None=WorldPosition().XYZ) {b=true}
  n71 = ActorPositionWS()
  n72 = NamedRerouteUsage()
  n73 = ComponentMask(None=StaticSwitchParameter(True=@n71, False=ShadingPathSwitch(Default=@n72, Deferred=@n72, Forward=@n71, Mobile=@n71)) {group=Base, parameterName="NoRVT (Top and Textures)", sortPriority=-4}) {b=true}
  n74 = ScalarParameter() {defaultValue=1300, group=Dust, parameterName=FadeColorsDistance, sortPriority=6}
  n75 = ScalarParameter() {group=Dust, parameterName=FadeColorDens}
  n76 = BlendMaterialAttributes(
   A=StaticSwitchParameter(
    True=SetMaterialAttributes!(MaterialAttributes=@n68.MaterialAttributes, Roughness=@n69) {attributeSetTypes [#4]}
    False=@n67
   ) {group=Ramp, parameterName=IsRamp}
   B=@n42
   Alpha=NamedRerouteDeclaration(None=SphereMask(A=@n70, B=@n73, Radius=@n74, Hardness=@n75)) {name=TopBottom Mask, nodeColor=<rgba 1 0 0.8672 1>}
  )
  n77 = ComponentMask(None=TextureCoordinate()) {r=true}
  n78 = Min(A=@n77, B=OneMinus(None=@n77))
  n79 = Subtract(A=WorldPosition().XYZ, B=ActorPositionWS())
  n80 = Constant() {r=-512}
  n81 = Divide(A=ComponentMask(None=Add(A=@n79, B=@n80)) {b=true}, B=Constant() {r=1000})
  n82 = Multiply(
   A=Add(A=Multiply(A=@n78, B=@n24), B=Saturate(None=@n81))
   B=Max(A=VertexColor().G, B=ScalarParameter() {defaultValue=0.2, group=Dirt, parameterName=MinDirt})
  )
  n83 = NamedRerouteUsage()
  n84 = LinearInterpolate(
   A=ScalarParameter() {defaultValue=-0.5, group=Dust, parameterName="Dust Min "}
   B=ScalarParameter() {defaultValue=4, group=Dust, parameterName=Dust Max}
   Alpha=@n25.G
  )
  n85 = BlendMaterialAttributes(
   A=@n26
   B=@n76
   Alpha=MaterialFunctionCall(
    Transition Phase=StaticSwitchParameter(True=@n82, False=@n24) {group=Ramp, parameterName=IsRamp}
    Height Texture=Multiply(A=@n83, B=@n84)
    Contrast=Constant() {r=0.5}
   ) {materialFunction=&HeightLerp}.Alpha
  ) {vertexAttributeBlendType=UseA}
  n86 = NamedRerouteUsage()
  n87 = TextureObjectParameter() {group=Base, parameterName=Diffuse, texture=/Game/Environments/Apollo/Sets/Tower/Textures/T_SmallTower_001_D}
  n88 = TextureSampleParameter2D(UVs=@n86) {
   group=Base
   parameterName=Mask
   samplerSource=SSM_Wrap_WorldGroupSettings
   samplerType=SAMPLERTYPE_Masks
   sortPriority=1
   texture=&T_SmallTower_001_MSK
  }
  n89 = TextureSampleParameter2D(UVs=@n86) {
   group=Base
   parameterName=Normal
   samplerSource=SSM_Wrap_WorldGroupSettings
   samplerType=SAMPLERTYPE_Normal
   sortPriority=2
   texture=&T_SmallTower_001_N
  }
  n90 = SetMaterialAttributes!(
   MaterialAttributes
   Base Color=TextureSample(UVs=@n86, Tex=@n87) {texture=/Game/Environments/Apollo/Sets/DarkVillage/Textures/T_DarkVillage_Floor_Stone_01_D}.RGB
   Metallic=@n88.G
   Specular=@n88.G
   Roughness=@n88.G
   Normal=Transform(None=@n89.RGB)
  ) {attributeSetTypes [
    #1
    #3
    #5
    #4
    #2
   ]}
  n91 = TextureObjectParameter() {group=Base, parameterName=Mask, samplerType=SAMPLERTYPE_Masks, texture=&T_SmallTower_001_MSK}
  n92 = ScalarParameter() {defaultValue=355, parameterName=GroundTextureSize}
  n93 = TextureObjectParameter() {group=Dirt, parameterName=Normal, samplerType=SAMPLERTYPE_Normal, texture=&T_SmallTower_001_N}
  n94 = MaterialFunctionCall!(
   Diffuse Texture Object=@n87
   Spec Texture Object=@n91
   TextureSize=@n92
   --------------------------
   Use seperate settings for Z projection
   Use seperate texture for Z projection
   Texture Size Z projection
   Z Plane Diffuse Texture Object
   Z Plane Spec Texture Object
   "--------------------------      "
   ProjectionTransitionContrast
   Use Texture Alpha Heightmap to Refine the Transistion
   HeightLerpTransitionContrast
   PickTheDiffuseHeightChannel
   TheDiffuseHeightChannel
   "--------------------------       "
   NormalTextureObject=@n93
   World Space Normals
   "--------------------------  "
   Export Float 4 Diffuse=StaticBool() {value=true}
   Export Float 4 Spec
   "--------------------------     "
   WorldPosition
   World Space Normal
  ) {materialFunction=/Engine/Functions/Engine_MaterialFunctions02/Texturing/WorldAlignedTextures_Complex}
  n95 = Constant() {r=4}
  n96 = MaterialFunctionCall!(World Position=WorldPosition().XYZ) {materialFunction=/Engine/Functions/Engine_MaterialFunctions02/Utility/DepthFromWorldPosition}
  n97 = Constant() {r=2500}
  n98 = Saturate(None=Divide(A=Subtract(A=@n96.Result, B=@n97), B=Constant() {r=-2500}))
  n99 = Constant() {r=1}
  n100 = Multiply(
   A=Constant() {r=10}
   B=LinearInterpolate(Alpha=MaterialFunctionCall() {materialFunction=&DitherTemporalAA}.Result) {constA=1, constB=1.5}
  )
  n101 = Divide(A=ComponentMask(None=WorldPosition().XYZ) {g=true, r=true}, B=@n92)
  n102 = MaterialFunctionCall!(
   Heightmap Texture=@n87
   Height Ratio=ScalarParameter() {defaultValue=0.08, group=POM, parameterName=HR}
   Min Steps=Add(A=Multiply(A=@n95, B=@n98), B=@n99)
   Max Steps=Add(A=@n99, B=Multiply(A=@n98, B=@n100))
   UVs=@n101
   Heightmap Channel
   Reference Plane=ScalarParameter() {group=POM, parameterName=RefeerncePlane}
   -----------------------------
   Use World Coordinates=StaticBool() {value=true}
   Specify Manual Texture Size
   Manual Texture Size
   ------------------------------
   "Render Shadows (Occlusion Mapping)"
   Light Vector
   Shadow Steps
   Shadow Penumbra
   Transform To VertexNormal
  ) {materialFunction=/Engine/Functions/Engine_MaterialFunctions01/Texturing/ParallaxOcclusionMapping}
  n103 = TextureSampleParameter2D(UVs=@n102.Parallax UVs) {
   group=Base
   parameterName=Mask
   samplerSource=SSM_Wrap_WorldGroupSettings
   samplerType=SAMPLERTYPE_Masks
   sortPriority=1
   texture=&T_SmallTower_001_MSK
  }
  n104 = MaterialFunctionCall(Float3=QualitySwitch(Default=@n94.Z Spec, Low=@n94.Z Spec, High=@n94.Z Spec, Medium=@n94.Z Spec, Epic=@n103.RGB)) {materialFunction=&BreakOutFloat3Components}
  n105 = NamedRerouteDeclaration(None=@n101) {name=GroundBricksUV, nodeColor=<rgba 0.6406 1 0 1>}
  n106 = TextureSampleParameter2D!(UVs=@n102.Parallax UVs, "DDX(UVs)"=DDY(Value=@n105), "DDY(UVs)"=DDY(Value=@n105), Apply View MipBias) {
   group=Base
   mipValueMode=TMVM_Derivative
   parameterName=Normal
   samplerSource=SSM_Wrap_WorldGroupSettings
   samplerType=SAMPLERTYPE_Normal
   sortPriority=2
   texture=&T_SmallTower_001_N
  }
  n107 = QualitySwitch(
   Default=@n94.Z Diffuse Texture
   Low=@n94.Z Diffuse Texture
   High=@n94.Z Diffuse Texture
   Medium=@n94.Z Diffuse Texture
   Epic=TextureSample(UVs=@n102.Parallax UVs, Tex=@n87).RGBA
  )
  n108 = SetMaterialAttributes(
   Base Color=Multiply(
    A=ComponentMask(None=@n107) {b=true, g=true, r=true}
    B=VectorParameter() {group=Base, parameterName=ToneBottom, sortPriority=-1, defaultValue=<rgba 0.1042 0.1014 0.0975 0>}.RGB
   )
   Metallic=@n104.G
   Specular=@n104.G
   Roughness=LinearInterpolate(
    A=ScalarParameter() {group=Base, parameterName=Rough Min, sortPriority=4}
    B=ScalarParameter() {defaultValue=1, group=Base, parameterName=Rough Max, sortPriority=4}
    Alpha=@n104.B
   )
   Normal=QualitySwitch(Default=@n94.Normal Z Texture, Low=@n94.Normal Z Texture, High=@n94.Normal Z Texture, Medium=@n94.Normal Z Texture, Epic=@n106.RGB)
  ) {attributeSetTypes [
    #1
    #3
    #5
    #4
    #2
   ]}
  n109 = ScalarParameter() {defaultValue=256, group=Dirt, parameterName=ScaleDirt}
  n110 = StaticBool() {value=true}
  n111 = MaterialFunctionCall(
   TextureObject=TextureObjectParameter() {group=Dirt, parameterName=DirtD, texture=/Game/Environments/Apollo/Sets/Terrain/Textures/T_Apollo_Dirt_D_01}
   TextureSize=@n109
   Export Float 4=@n110
  ) {materialFunction=&WorldAlignedTexture}
  n112 = VectorParameter() {group=Dust, parameterName=Dust, defaultValue=<rgba 0.276 0.1329 0.0701 0>}
  n113 = Saturate(None=OneMinus(None=NamedRerouteUsage()))
  n114 = Multiply(
   A=ComponentMask(None=@n111.Z Texture) {b=true, g=true, r=true}
   B=Multiply(
    A=LinearInterpolate(A=@n112.RGB, Alpha=@n113)
    B=VectorParameter() {group=Dirt, parameterName=DirtCracks, defaultValue=<rgba 0.4193 0.4193 0.4193 0>}.RGB
   )
  )
  n115 = MaterialFunctionCall(Float4=@n111.Z Texture) {materialFunction=&BreakOutFloat4Components}
  n116 = TextureObjectParameter() {group=Dirt, parameterName=DirtN, samplerType=SAMPLERTYPE_Normal, texture=/Game/Environments/Apollo/Sets/Terrain/Textures/T_Apollo_Dirt2_N}
  n117 = SetMaterialAttributes(
   Base Color=@n114
   Specular=@n115.R
   Roughness=Multiply(A=@n115.R, B=ScalarParameter() {defaultValue=3, group=Dirt, parameterName=DirtR})
   Normal=NamedRerouteDeclaration(None=MaterialFunctionCall(TextureObject=@n116, TextureSize=@n109, WorldSpace=@n110) {materialFunction=&WorldAlignedNormal}.Z Texture) {name=DirtNormal, nodeColor=<rgba 0.3828 1 0 1>}
  ) {attributeSetTypes [
    #1
    #3
    #5
    #4
    #2
   ]}
  n118 = ScalarParameter() {defaultValue=0.1, group=Dirt, parameterName=TileToDirtTrans}
  n119 = ScalarParameter() {defaultValue=1, group=Dirt, parameterName=TileToDirtTrans2}
  n120 = NamedRerouteUsage()
  n121 = ComponentMask(None=@n107) {a=true}
  n122 = BlendMaterialAttributes(
   A=@n108
   B=@n117
   Alpha=MaterialFunctionCall!(
    Diffuse 1=@n114
    Diffuse 2=Constant3Vector() {constant=<rgba 1 0 0 0>}
    Transistion Phase=LinearInterpolate(A=@n118, B=@n119, Alpha=@n120)
    Height Texture 1=NamedRerouteDeclaration(None=@n121) {name=GroundTilesHeight, nodeColor=<rgba 0.4062 1 0 1>}
    Height Texture 2=@n115.A
    Contrast=ScalarParameter() {defaultValue=-25, group=Dirt, parameterName=TileToDirtCon}
   ) {materialFunction=/Engine/Functions/Engine_MaterialFunctions02/Texturing/HeightLerpWithTwoHeightMaps}.Alpha
  ) {vertexAttributeBlendType=UseA}
  n123 = QualitySwitch(Default=@n90, Low=@n90, High=@n122, Medium=@n122, Epic=@n122)
  n124 = TextureSample() {samplerType=SAMPLERTYPE_Masks, texture=/Game/Environments/Apollo/Sets/Terrain/Textures/T_Apollo_Cliff_S_01}
  n125 = SetMaterialAttributes(
   Base Color=Multiply(
    A=VectorParameter() {parameterName=WallTone, defaultValue=<rgba 0.4297 0.3482 0.2885 0>}.RGB
    B=TextureSample() {texture=/Game/Environments/Apollo/Sets/Terrain/Textures/T_Apollo_Cliff_D_01}.RGB
   )
   Metallic=@n124.G
   Specular=@n124.G
   Roughness=Multiply(A=@n124.B, B=Constant() {r=0.6})
   Normal=Transform(None=TextureSample() {samplerType=SAMPLERTYPE_Normal, texture=/Game/Environments/Apollo/Sets/Terrain/Textures/T_Apollo_Cliff_N_01}.RGB)
  ) {attributeSetTypes [
    #1
    #3
    #5
    #4
    #2
   ]}
  n126 = ComponentMask(None=VertexNormalWS()) {b=true}
  n127 = OneMinus(None=NamedRerouteUsage())
  n128 = Constant() {r=3}
  n129 = StaticSwitchParameter(
   True=BlendMaterialAttributes(
    A=ShadingPathSwitch(Default=@n123, Deferred=@n123, Forward=@n123, Mobile=@n90)
    B=@n125
    Alpha=MaterialFunctionCall(Transistion Phase=@n126, Height Texture 1=@n127, Height Texture 2=@n115.A, Contrast=@n128) {materialFunction=&HeightLerpWithTwoHeightMaps}.Alpha
   ) {vertexAttributeBlendType=UseA}
   False=ShadingPathSwitch(Default=@n85, Deferred=@n85, Forward=@n76, Mobile=@n76)
  ) {group=Ramp, parameterName=IsGround}
  n130 = GetMaterialAttributes(None=@n129) {attributeGetTypes [#1, #3, #4]}
  n131 = VectorParameter() {group=Gradient, parameterName=Gradient Bottom, defaultValue=<rgba 0.5911 0.5694 0.5041 0>}
  n132 = VectorParameter() {group=Gradient, parameterName=Gradient Top, defaultValue=<rgba 0.8021 0.8021 0.8021 0>}
  n133 = NamedRerouteUsage()
  n134 = ScalarParameter() {defaultValue=1, parameterName=Roughness Bottom}
  n135 = SetMaterialAttributes!(
   MaterialAttributes=@n130.MaterialAttributes
   Base Color=Multiply(A=LinearInterpolate(A=@n131.RGB, B=@n132.RGB, Alpha=@n133), B=@n130.Base Color)
   Metallic=Multiply(A=@n133, B=@n130.Metallic)
   Roughness=LinearInterpolate(A=Multiply(A=@n130.Roughness, B=@n134), B=@n130.Roughness, Alpha=@n133)
  ) {attributeSetTypes [#1, #3, #4]}
  n136 = GetMaterialAttributes(None=StaticSwitchParameter(True=@n129, False=StaticSwitchParameter(True=@n135, False=@n129) {group=Base, parameterName=use Gradient, sortPriority=-4}) {group=Base, parameterName="NoRVT (Top and Textures)", sortPriority=-4}) {attributeGetTypes [#1]}
  n137 = CollectionParameter() {
   collection=&FortniteMaterialParameters
   parameterName=FortniteActiveTimeOfDay
  }
  n138 = Constant4Vector() {constant=<rgba 0 0 0 1>}
  n139 = MaterialFunctionCall(
   Base=LinearInterpolate(B=@n136.Base Color) {constA=0.5, constAlpha=0.75}
   Blend=Multiply(A=VectorParameter() {group=Glow, parameterName=Glow, defaultValue=<rgba 0.5586 0.2479 0.8359 0>}.RGB, B=DotProduct(A=@n137, B=@n138.RGBA))
  ) {materialFunction=&Blend_Overlay}
  n140 = OneMinus(None=Fresnel() {baseReflectFraction=0, exponent=1.3})
  n141 = TextureSampleParameter2D!(UVs=NamedRerouteUsage(), Apply View MipBias) {
   group=Base
   parameterName=Mask
   samplerSource=SSM_Wrap_WorldGroupSettings
   samplerType=SAMPLERTYPE_Masks
   sortPriority=1
   texture=&T_SmallTower_001_MSK
  }
  n142 = Multiply(A=Multiply(A=Multiply(A=@n139.Result, B=@n140), B=OneMinus(None=@n141.G)), B=NamedRerouteUsage())
  n143 = StaticSwitchParameter(
   True=@n136.MaterialAttributes
   False=StaticSwitchParameter(
    True=SetMaterialAttributes!(MaterialAttributes=@n136.MaterialAttributes, Emissive Color=@n142) {attributeSetTypes [#6]}
    False=@n136.MaterialAttributes
   ) {group=Glow, parameterName=GlowBits}
  ) {group=Base, parameterName="NoRVT (Top and Textures)", sortPriority=-4}
  n144 = LinearInterpolate(
   A=ScalarParameter() {defaultValue=8}
   B=ScalarParameter() {parameterName=RVT_Mip_Ground}
   Alpha=ComponentMask(None=VertexNormalWS()) {b=true}
  )
  n145 = ScalarParameter() {defaultValue=4, parameterName=RVTMip}
  n146 = RuntimeVirtualTextureSampleParameter!(
   UVs
   World Position
   Mip Bias=StaticSwitchParameter(True=Constant(), False=StaticSwitchParameter(True=@n144, False=@n145) {group=Ramp, parameterName=IsGround}) {group=Ramp, parameterName=IsRamp}
  ) {
   group=RVT
   materialType=BaseColor_Normal_Specular_YCoCg
   mipValueMode=RVTMVM_MipBias
   parameterName=RVTColor
   virtualTexture=/Game/Athena/Apollo/Maps/Landscape/VT/RVT_Apollo_Landscape
  }
  n147 = SphereMask(A=@n2, B=@n3, Radius=ScalarParameter() {defaultValue=100, group=RVT, parameterName=RVTFadeDistanceNormals}, Hardness=@n4)
  n148 = SetMaterialAttributes!(
   MaterialAttributes=BlendMaterialAttributes(
    A=@n143
    B=SetMaterialAttributes!(MaterialAttributes=@n143, Normal=@n146.Normal) {attributeSetTypes [#2]}
    Alpha=Saturate(None=@n147)
   )
   Base Color=@n146.BaseColor
   Metallic=Constant()
   Specular=@n146.BaseColor
   Roughness=@n146.BaseColor
  ) {attributeSetTypes [
    #1
    #3
    #5
    #4
   ]}
  n149 = BlendMaterialAttributes(
   A=@n143
   B=@n148
   Alpha=MaterialFunctionCall(Transition Phase=Saturate(None=@n18), Height Texture=LinearInterpolate(A=@n19, B=@n20, Alpha=@n21.R)) {materialFunction=&HeightLerp}.Lerp Alpha No Contrast
  ) {vertexAttributeBlendType=UseA}
  n150 = StaticSwitchParameter(
   True=@n143
   False=MaterialFunctionCall(NoRVT=@n143, RVT=MaterialProxyReplace(Realtime=@n149, MaterialProxy=@n143)) {materialFunction=&FN_RVT_Switch_Material}.Result
  ) {group=Base, parameterName="NoRVT (Top and Textures)", sortPriority=-4}
  n151 = ShadingPathSwitch(Default=@n150, Deferred=@n150, Forward=@n143, Mobile=@n143)
  n152 = ScalarParameter() {defaultValue=0.2, group=Emissive, parameterName=FresnelMin}
  n153 = VertexNormalWS()
  n154 = CameraVectorWS()
  n155 = Saturate(None=Power(Base=DotProduct(A=@n153, B=@n154), Exp=ScalarParameter() {defaultValue=3, group=Emissive, parameterName=EmissiveFresnel}))
  n156 = TextureSampleParameter2D(UVs=StaticSwitchParameter(True=TextureCoordinate() {coordinateIndex=1}, False=NamedRerouteUsage()) {group=Emissive, parameterName=Emissive UV1, sortPriority=-2}) {
   group=Base
   parameterName=Mask
   samplerSource=SSM_Wrap_WorldGroupSettings
   samplerType=SAMPLERTYPE_Masks
   sortPriority=1
   texture=&T_SmallTower_001_MSK
  }
  n157 = ScalarParameter() {defaultValue=4, group=Emissive, parameterName=EmissiveNoiseScale, sortPriority=1}
  n158 = StaticSwitchParameter(True=TextureCoordinate() {coordinateIndex=1}, False=TextureCoordinate()) {group=Emissive, parameterName=Emissive UV1, sortPriority=-2}
  n159 = Multiply(A=@n157, B=@n158)
  n160 = Multiply(A=Time(), B=ScalarParameter() {defaultValue=2, group=Emissive, parameterName=EmissiveSpeed, sortPriority=2})
  n161 = TextureSample(UVs=Panner(Coordinate=Multiply(A=@n157, B=@n158)) {speedX=0.1, speedY=0.1}) {samplerSource=SSM_Wrap_WorldGroupSettings, samplerType=SAMPLERTYPE_Grayscale, texture=&LowResBlurredNoise}
  n162 = TextureSample(UVs=BumpOffset(Coordinate=Panner(Coordinate=@n159, Time=@n160) {speedX=0.15, speedY=0.075}, Height=@n161.R, HeightRatioInput=Constant() {r=0.5})) {samplerSource=SSM_Wrap_WorldGroupSettings, samplerType=SAMPLERTYPE_Grayscale, texture=&LowResBlurredNoise}
  n163 = Multiply(A=@n159, B=Constant() {r=0.8})
  n164 = TextureSample(UVs=BumpOffset(Coordinate=Panner(Coordinate=@n163, Time=@n160) {speedX=-0.13, speedY=-0.08}, Height=@n161.R, HeightRatioInput=Constant() {r=0.2})) {samplerSource=SSM_Wrap_WorldGroupSettings, samplerType=SAMPLERTYPE_Grayscale, texture=&LowResBlurredNoise}
  n165 = LinearInterpolate(
   A=VectorParameter() {group=Emissive, parameterName=EmissiveColor, sortPriority=1}.RGB
   B=VectorParameter() {group=Emissive, parameterName=EmissiveColor2, sortPriority=1}.RGB
   Alpha=Saturate(None=Multiply(A=@n162.R, B=@n164.R))
  )
  n166 = Constant()
  n167 = SetMaterialAttributes!(
   MaterialAttributes=@n151
   Emissive Color=Multiply(A=LinearInterpolate(A=@n152, Alpha=@n155), B=Multiply(A=@n156.B, B=@n165))
   Base Color=@n166
   Metallic=@n166
   Specular=Constant() {r=0.05}
   Roughness=Constant() {r=0.2}
  ) {attributeSetTypes [
    #6
    #1
    #3
    #5
    #4
   ]}
  n168 = MaterialFunctionCall!(In=StaticSwitchParameter(True=BlendMaterialAttributes(A=@n151, B=@n167, Alpha=@n156.B), False=@n151) {group=Emissive, parameterName=Emissive, sortPriority=-1}) {materialFunction=&/MF_TowerEvent}
  Reroute(None=StaticSwitchParameter(
    True=SetMaterialAttributes!(MaterialAttributes=@n168.Result, World Position Offset=@n16) {attributeSetTypes [#7]}
    False=@n168.Result
   ) {parameterName=UseWPOWings})
  VertexInterpolator(VS=@n11)
  VirtualTextureFeatureSwitch!(No, Yes)
 ]
}
